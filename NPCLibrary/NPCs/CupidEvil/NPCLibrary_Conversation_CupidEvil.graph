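# Conversation database for the CupidEvil NPC, defining the dialogue entered through its CupidEvil entry tag.
# The entry leads to one speech node with a default taunt and prioritized contextual lines for a first-ever conversation or for players wearing specific outfits.
# That speech branches through link nodes to shared conversations for NPC library quests, basic services, and a Season 26 Snapshot quest.
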ConversationDatabase /NPCLibrary/NPCs/CupidEvil/NPCLibrary_Conversation_CupidEvil
entry_points [{tag=Conversation.Entry.CupidEvil, destinations [#1]}]
nodes [
 ConversationEntryPointNode_7 {
  id=#1
  class=ConversationEntryPointNode
  outputs [#2]
  properties {entryTag {tagName=Conversation.Entry.CupidEvil}}
 }
 Speech_C_4 {
  id=#2
  class=Speech_C
  outputs [#3, #4, #5]
  properties {
   nodeName=""
   generalConfig {
    defaultMessage=SPOOKY SHADOWS HAUNT THIS ISLAND. I WILL SCATTER THEM WITH MY FLAME.
    contextualMessages [
     {
      message=DO NOT DARE TO CHALLENGE ME. YOU WILL LOSE.
      contextRequirements [{requirement {class=FortControllerRequirement_IsFirstEverConversationWithNPC}}]
      priority {curve {rowName=Tandem.Speech.Priority.FirstEverInteraction}}
     }
     {
      message="I AM YOUR SPOOKY FUTURE. BOW BEFORE ME!"
      contextRequirements [{requirement {
         class=FortControllerRequirement_HasCID
         properties {allowedCIds [/BRCosmetics/Athena/Items/Cosmetics/Characters/CID_070_Athena_Commando_M_Cupid]}
        }}]
      priority {curve {rowName=Tandem.Speech.Priority.OutfitReaction}}
     }
     {
      message=I HAVE ALLOWED THE FLAME TO CONSUME MY HEART... AND IT HAS MADE ME POWERFUL.
      contextRequirements [{requirement {
         class=FortControllerRequirement_HasCID
         properties {allowedCIds [/BRCosmetics/Athena/Items/Cosmetics/Characters/CID_523_Athena_Commando_F_Cupid]}
        }}]
      priority {curve {rowName=Tandem.Speech.Priority.OutfitReaction}}
     }
     {
      message=YOUR SPOOKY MISCHIEF IS NO MATCH FOR THE LIGHT OF MY BURNING HEART.
      contextRequirements [{requirement {class=FortControllerRequirement_HasCID, properties {allowedCIds [/BRCosmetics/Athena/Items/Cosmetics/Characters/Character_EctoCat]}}}]
      priority {curve {rowName=Tandem.Speech.Priority.OutfitReaction}}
     }
     {
      message="MY LIEGE... Oh, um, sorry. My liege, what would you ask of me?"
      contextRequirements [{requirement {
         class=FortControllerRequirement_HasCID
         properties {allowedCIds [/BRCosmetics/Athena/Items/Cosmetics/Characters/CID_A_133_Athena_Commando_M_DarkVikingFire]}
        }}]
      priority {curve {rowName=Tandem.Speech.Priority.OutfitReaction}}
     }
    ]
   }
  }
 }
 ConversationLinkNode_2 {
  id=#5
  class=ConversationLinkNode
  properties {remoteEntryTag {tagName=Conversation.Entry.Season26.Snapshot.P03.Q07}}
 }
 ConversationLinkNode_0 {id=#4, class=ConversationLinkNode, properties {remoteEntryTag {tagName=Conversation.Entry.BasicServices}}}
 ConversationLinkNode_1 {id=#3, class=ConversationLinkNode, properties {remoteEntryTag {tagName=Conversation.Entry.NPCLibrary.Quests}}}
]
